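; Interactable TV scare actor that spawns the WildEstate Troop Factor boss when a player interacts or, by random chance and line trace, comes near.
; Interaction also applies an idle-knockback effect in range; the proximity handler copies the parent's, minus its destroy-actor call.
; Challenge-manager mode gates parent death handling and can remove the proximity trigger.
; Helpers queue an enemy stim on the boss AI, replicate emergence FX and debug-teleport players.
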(event EventOnDeathServer
  (params (Damage :type Float%20%28single-precision%29) (DamageTags :type Gameplay%20Tag%20Container%20Structure%20%28by%20ref%29) (Momentum :type Vector :default "0, 0, 0") (HitInfo :type Hit%20Result%20Structure%20%28by%20ref%29) (InstigatedBy :type Controller%20Object%20Reference) (DamageCauser :type Actor%20Object%20Reference) (EffectContext :type Gameplay%20Effect%20Context%20Handle%20Structure))
  (if (Variables|Default|GetChallengeManagerActive)
    (|CallParentFunction Damage DamageTags Momentum HitInfo InstigatedBy DamageCauser EffectContext)))

(event Custom|InitForChallengeManager
  (Variables|Default|SetChallengeManagerActive true)
  (|CallParentFunction)
  (if (not (ScalableFloat|GetValueasBool (Variables|Default|GetIsOverlapSpawnable)))
    (Components|DestroyComponent (Variables|Default|GetSpawnProximityTrigger))))

(event EventBeginPlay
  (Rendering|SetRenderCustomDepth (Variables|Default|GetFORT_TV) false))

(event Collision|EventActorBeginOverlap
  (params (OtherActor :type Actor%20Object%20Reference)))

(event EventTick
  (params (DeltaSeconds :type Float%20%28single-precision%29)))

(event EventAISpawned
  (params (SpawnedActor :type Actor%20Object%20Reference) (bResult :type Boolean))
  (bind _asbp_player_pawn_non_participant_boss_wild_estate_troop_factor (Utilities|Casting|CastToBP_PlayerPawn_NonParticipant_Boss_WildEstate_TroopFactor SpawnedActor))
  (bind _output_get (Variables|Default|SetSpawnedBoss _asbp_player_pawn_non_participant_boss_wild_estate_troop_factor))
  (Class|BPPlayerPawnNonParticipantBossWildEstateTroopFactor|IgnoreActorCollision _output_get self)
  (Components|Activation|Deactivate (Variables|Default|GetAthenaSpecialActor)))

(event Interaction|EventBlueprintOnInteract
  (params (InteractingPawn :type Fort%20Pawn%20Object%20Reference) (InteractionBeingAttempted :type EInteractionBeingAttempted%20Enum :default "FirstInteraction"))
  (Variables|InternalData|SetTargetPawn InteractingPawn)
  (Variables|Default|SetInteractionSuccessful true)
  (if (not (Variables|InternalData|GetShouldSpawnRewards))
    (Utilities|FlowControl|Delay (ScalableFloat|GetValueatLevel (Variables|Default|GetAnticipationDelay))))
  (CallFunction|TrySpawnAI)
  (CallFunction|ApplyImpulse))

(event Custom|ApplyImpulse
  (if (ScalableFloat|GetValueasBool (Variables|Default|GetAllowImpulse))
    (Fortnite|ApplyGameplayEffectToActorsInRange (Utilities|Array|MakeArray "/WildEstate_TroopFactor/Gameplay/Abilites/GE_WildEstate_TroopFactor_GrantIdleKnockback.GE_WildEstate_TroopFactor_GrantIdleKnockback_C") (Utilities|Array|MakeArray "ObjectTypeQuery3") (Transformation|GetActorLocation self) (ScalableFloat|GetValueatLevel (Variables|Default|GetImpulseApplyRange)) false 0 (Ability|GetAbilitySystemComponent self))))

(event Custom|StimTarget
  (bind _asfort_athena_aibot_controller (Utilities|Casting|CastToFortAthenaAIBotController (AI|GetAIController (Variables|Default|GetSpawnedBoss))))
  (Class|FortAthenaAIBotController|QueueStim _asfort_athena_aibot_controller (Variables|InternalData|GetTargetPawn) (Transformation|GetActorLocation (Variables|InternalData|GetTargetPawn)) (Transformation|GetActorLocation (Variables|InternalData|GetTargetPawn)) "Enemy" 1e+02 "None"))

(event Custom|Start%20Emergence%20FX
  (Networking|SwitchHasAuthority
    (:Authority
      (Variables|Default|SetEnableEmergenceFX true))))

(event Custom|DebugTeleportToTroopFactorTV
  ;[0.03932 0.04454 0.15 0.5] Debug cheat to warp players to TV location on map
  (bind _outfortplayerpawns (Fortnite|GetFortPlayerPawns))
  (for _array_element _outfortplayerpawns
    (if (Pawn|IsPlayerControlled _array_element)
      (Transform|Actor|TeleportPlayerPawn _array_element (Transformation|GetActorLocation)))))

(event OnComponentBeginOverlap(SpawnProximityTrigger)
  (params (OverlappedComponent :type Primitive%20Component%20Object%20Reference) (OtherActor :type Actor%20Object%20Reference) (OtherComp :type Primitive%20Component%20Object%20Reference) (OtherBodyIndex :type Integer) (bFromSweep :type Boolean) (SweepResult :type Hit%20Result%20Structure%20%28by%20ref%29))
  ;[1 1 1] Pawn is too close, try to spawn the AI -------- (Copied from parent to remove destroy actor call)
  (Networking|SwitchHasAuthority
    (:Authority
      (bind _asfort_pawn (Utilities|Casting|CastToFortPawn OtherActor))
      (if (<= (Math|Random|RandomFloatinRange 0.0 1.0) (ScalableFloat|GetValueatLevel (Variables|Default|GetChanceToSpawnOnProximity)))
        (bind (_outhit _returnvalue) (Collision|LineTraceByChannel (+ (Transformation|GetActorLocation self) "0.000000,0.000000,50.000000") (Transformation|GetActorLocation _asfort_pawn) "TraceTypeQuery1" false (Utilities|Array|MakeArray self) "None" true "(R=1.000000,G=0.000000,B=0.000000,A=1.000000)" "(R=0.000000,G=1.000000,B=0.000000,A=1.000000)" 3e+01))
        (if _returnvalue
          (bind _asfort_pawn_2 (Utilities|Casting|CastToFortPawn (.HitActor (Collision|BreakHitResult _outhit))))
          (Variables|InternalData|SetSpawningByProximity true)
          (Variables|InternalData|SetTargetPawn _asfort_pawn_2)
          (if (not (Variables|InternalData|GetCanTriggerProximityAndInteraction))
            (Utilities|FlowControl|Delay 0.2))
          (Variables|Default|SetInteractionSuccessful true)
          (CallFunction|TrySpawnAI))
        (else
          (Components|DestroyComponent (Variables|Default|GetSpawnProximityTrigger)))))))

(fn BlueprintCanInteract
  (params (InteractingPawn :type Fort%20Pawn%20Object%20Reference) (InteractionBeingAttempted :type EInteractionBeingAttempted%20Enum :default "FirstInteraction") (InteractionType :type TInteractionType%20Enum :default "IT_NoInteraction"))
  (return (not (or (Variables|Default|GetInteractionSuccessful) (Variables|InternalData|GetSpawningByProximity)))))

(fn OnRep_EnableEmergenceFX
  (if (Variables|Default|GetEnableEmergenceFX)
    (Components|Activation|Activate (Variables|Default|GetEmergeEffect) true)
    (Components|Activation|Deactivate (Variables|Default|GetTV_FX))))

(fn ConstructionScript
  (|CallParentFunction))
